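; Swarm grenade area actor: a timed insect-swarm hazard with lifespan from a scalable float.
; The authority registers an AI danger zone and unregisters it on end or death.
; Clients orient the Niagara swarm by traces to surface or water, or fixed rotation for tagged instigators.
; Clients also fade looping-area audio and, after an effects-quality check, spawn a wash-force field and timeline, winding everything down before expiry.
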(event EventBeginPlay
  (bind _returnvalue (Class|AthenaAIServiceZoneManager|GetAthenaAIServiceZoneManager true))
  (bind _returnvalue_2 (Transformation|GetActorLocation))
  (bind _sf_lifespan (Variables|Default|GetSF_LifeSpan))
  (bind _returnvalue_3 (ScalableFloat|GetValueatLevel _sf_lifespan))
  (bind _returnvalue_4 (+ _returnvalue_2 "0.000000,0.000000,-100.000000"))
  (bind _returnvalue_5 (+ _returnvalue_2 "0.000000,0.000000,100.000000"))
  (Actor|SetLifeSpan _returnvalue_3)
  (Networking|SwitchHasAuthority
    (:Authority
      ;[1 1 1] Make sure AI knows it's dangerous here
      (Utilities|IsValid _returnvalue
        (:Is%20Valid
          (Class|AthenaAIServiceZoneManager|RegisterDangerZone _returnvalue self (ScalableFloat|GetValueatLevel (Variables|Default|GetSF_DamageRadius))))))
    (:Remote
      (Utilities|FlowControl|Sequence
        (:then_0
          (bind _returnvalue_6 (GameplayTags|HasMatchingGameplayTag(Message) (Variables|Actor|GetInstigator) "(TagName=\"Athena.AI.IS\")"))
          (Effects|Components|ParticleSystem|SetBoolParameter (Variables|Default|GetNS_AI_Wildlife_InsectSwarm) "InstigatorIsSwarm" _returnvalue_6)
          (if _returnvalue_6
            (Transformation|SetWorldRotation (Variables|Default|GetNS_AI_Wildlife_InsectSwarm) "90.000000,0.000000,0.000000" false false)
            (else
              (bind (_outhit _returnvalue_7) (Collision|SphereTraceByChannel _returnvalue_5 _returnvalue_4 3e+01))
              (if _returnvalue_7
                (Transformation|SetWorldLocationAndRotation (Variables|Default|GetNS_AI_Wildlife_InsectSwarm) (.Location (Collision|BreakHitResult _outhit)) (Math|Conversions|RotationFromXVector (.ImpactNormal (Collision|BreakHitResult _outhit))) false false)
                (else
                  ;[1 1 1] Check if it was spawned in Water
                  (bind (_outhit_2 _returnvalue_8) (Collision|LineTraceByChannel _returnvalue_5 _returnvalue_4 "TraceTypeQuery12"))
                  (if _returnvalue_8
                    (Transformation|SetWorldLocationAndRotation (Variables|Default|GetNS_AI_Wildlife_InsectSwarm) (.Location (Collision|BreakHitResult _outhit_2)) (Math|Conversions|RotationFromXVector (.ImpactNormal (Collision|BreakHitResult _outhit_2))) false false)))))))
        (:then_1
          (bind _returnvalue_9 (Audio|SpawnSoundAttached (Variables|Default|GetAOESound) (Variables|Default|GetDefaultSceneRoot)))
          (bind _output_get (Variables|Default|SetAOEAudio _returnvalue_9))
          (Audio|Components|Audio|FadeIn _output_get 0.15 1.0 0.0 "Linear" :!class /Script/Engine.AudioComponent)
          (Utilities|FlowControl|Sequence
            (:then_0
              (Utilities|FlowControl|Delay (- _returnvalue_3 2.0))
              (Components|Activation|Deactivate (Variables|Default|GetNS_AI_Wildlife_InsectSwarm))
              (Utilities|FlowControl|Delay 0.9)
              (bind _aoeaudio (Variables|Default|GetAOEAudio)
                (:then
                  (Audio|Components|Audio|FadeOut _aoeaudio 1.25 0.0 "Linear" :!class /Script/Engine.AudioComponent))
                (:else)))
            (:then_1
              (Utilities|EffectsQualityCheck 0
                (:Cinematic
                  _)
                (:Epic
                  _))
              (bind _returnvalue_10 (Game|SpawnActorfromClass "/Game/Effects/Niagara/Blueprints/PhsyicsFields/FN_WashForce.FN_WashForce_C" (Transformation|GetActorTransform) "Undefined" "MultiplyWithRoot" 0 7e+02 0.0 -999.0 0))
              (Variables|Default|SetWashForce _returnvalue_10)
              (Components|Timeline|PlayfromStart (Variables|BGA_SwarmGrenade|GetTimeline))
              (Utilities|FlowControl|Delay (- _returnvalue_3 2.0))
              (Components|Timeline|ReversefromEnd (Variables|BGA_SwarmGrenade|GetTimeline))
              (Utilities|FlowControl|Delay 0.9)
              (bind _washforce (Variables|Default|GetWashForce)
                (:then
                  (Actor|DestroyActor _washforce))
                (:else)))))))))

(event EventEndPlay
  (params (EndPlayReason :type EEndPlayReason%20Enum :default "Destroyed"))
  (bind _returnvalue (Class|AthenaAIServiceZoneManager|GetAthenaAIServiceZoneManager true))
  (Actor|DestroyActor (Variables|Default|GetWashForce))
  (bind _aoeaudio (Variables|Default|GetAOEAudio)
    (:then
      (Audio|Components|Audio|FadeOut _aoeaudio 1.25 0.0 "Linear" :!class /Script/Engine.AudioComponent))
    (:else
      _))
  (Networking|SwitchHasAuthority
    (:Authority
      (Utilities|IsValid _returnvalue
        (:Is%20Valid
          (Class|AthenaAIServiceZoneManager|UnregisterDangerZone _returnvalue self))))))

(event EventOnDeathPlayEffects
  (params (Damage :type Float%20%28single-precision%29) (DamageTags :type Gameplay%20Tag%20Container%20Structure%20%28by%20ref%29) (Momentum :type Vector :default "0, 0, 0") (HitInfo :type Hit%20Result%20Structure%20%28by%20ref%29) (InstigatedBy :type Fort%20Pawn%20Object%20Reference) (DamageCauser :type Actor%20Object%20Reference) (EffectContext :type Gameplay%20Effect%20Context%20Handle%20Structure))
  (bind _returnvalue (Class|AthenaAIServiceZoneManager|GetAthenaAIServiceZoneManager true))
  (|CallParentFunction Damage DamageTags Momentum HitInfo InstigatedBy DamageCauser EffectContext)
  (bind _aoeaudio (Variables|Default|GetAOEAudio)
    (:then
      (Audio|Components|Audio|FadeOut _aoeaudio 1.25 0.0 "Linear" :!class /Script/Engine.AudioComponent))
    (:else
      _))
  (Networking|SwitchHasAuthority
    (:Authority
      (Utilities|IsValid _returnvalue
        (:Is%20Valid
          (Class|AthenaAIServiceZoneManager|UnregisterDangerZone _returnvalue self))))))

(fn ConstructionScript)
