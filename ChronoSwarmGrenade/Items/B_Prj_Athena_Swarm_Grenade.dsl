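; Thrown swarm grenade projectile that explodes on impact or water entry and spawns the BGA_SwarmGrenade actor.
; On hit, authority starts the short-timer spawn and explosion, while remotes play Niagara, rumble, camera shake and sound indicator.
; If already in water shortly after launch, it instead hides, drops a grenade pickup and expires.
; Hit and EnteredWater flags keep these paths exclusive; OnRep_EnteredWater selects splash or bounce sound.
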(event EventBeginPlay
  ;[1 1 1] On Begin Play, turn on Fire Particle FX for Burning Rag - Bind Event to Enter Water - On Water Enter, disable Fire Particle FX and then hide the projectile and spawn a new pickup
  (Components|Activation|Activate (Variables|Default|GetNiagara_Projectile_Particles) false)
  ;;[1 1 1] This is to fix throwing the projectile whilst in water, to allow it time to leave the pawns hands
  (Utilities|FlowControl|Delay 0.2)
  (Variables|Default|SetCanInteractWithWater true)
  (if (Water|IsinWater (Variables|Water|GetWaterInteractionComponent))
    (CallFunction|HasEnteredWater)))

(event Custom|HasEnteredWater
  ;[1 1 1] On Begin Play, turn on Fire Particle FX for Burning Rag - Bind Event to Enter Water - On Water Enter, disable Fire Particle FX and then hide the projectile and spawn a new pickup
  ;;[1 1 1] If already entered Hit, don't do anything with the water interaction. Set Entered Water so it can be checked in Hit interaction.
  (if (not (Variables|Default|GetHit))
    (Utilities|FlowControl|DoOnce false
      (:Completed
        (if (not (Variables|Default|GetEnteredWater))
          (Variables|Default|SetEnteredWater true)
          (Utilities|FlowControl|Sequence
            (:then_0
              (Components|Activation|Deactivate (Variables|Default|GetNiagara_Projectile_Particles))
              (Rendering|SetActorHiddenInGame true)
              (Inventory|SpawnPickupInWorld "/ChronoSwarmGrenade/Items/WID_Athena_Swarm_Grenade.WID_Athena_Swarm_Grenade" 1 (Transformation|GetWorldLocation (Variables|Collision|GetCapsuleComponent)) "0.000000,0.000000,0.000000" 0 true false)
              (Actor|SetLifeSpan 0.5))
            (:then_1
              (Audio|Components|Audio|FadeOut (Variables|B_Prj_Grenade|GetGrenadeFuse_AudioComponent) 0.2 0.0 "Linear" :!class /Script/Engine.AudioComponent))
            (:then_2
              (Effects|Components|ParticleSystem|SpawnFXAtLocation "/Game/Athena/Items/Consumables/MolotovCocktail/NS_FireflyJar_WaterSplash.NS_FireflyJar_WaterSplash" (Transformation|GetWorldLocation (Variables|Collision|GetCapsuleComponent))))))))))

(event OnEnterWaterEvent(WaterInteractionComponent)
  (params (WaterBody :type Fort%20Water%20Body%20Actor%20Object%20Reference) (WaterInteractionComponent :type Fort%20Water%20Interaction%20Component%20Object%20Reference) (bIsFirstBody :type Boolean))
  ;[1 1 1] On hit, store the Hit Location / Hit Normal then mirror the projectile velocity and hit normal to get the reflection vector
  (if bIsFirstBody
    (if (and (> (Actor|GetGameTimeSinceCreation) 0.2) (not (Variables|Default|GetEnteredWater)))
      (Variables|Default|SetEnteredWater true)
      (CallFunction|HeldWaterImpact)
      (Variables|Default|SetSpawnBGATransform (Math|Transform|MakeTransform (+ (Transformation|GetActorLocation self) "0.000000,0.000000,15.000000")))
      (Effects|Components|ParticleSystem|SpawnFXAtLocation "/Game/Athena/Items/Consumables/MolotovCocktail/NS_FireflyJar_WaterSplash.NS_FireflyJar_WaterSplash" (Transformation|GetWorldLocation (Variables|Collision|GetCapsuleComponent)))
      (Utilities|Time|SetTimerbyEvent (EventDispatchers|CreateEvent :!event Custom|Defer%20BGA%20Spawn) 0.1)
      (Delivery|DoExplosion))))

(event Custom|Defer%20BGA%20Spawn
  ;[1 1 1] Spawn the Swarm Actor Here
  (Building|SpawnBuildingGameplayActor "/ChronoSwarmGrenade/Items/BGA_SwarmGrenade.BGA_SwarmGrenade_C" (Variables|Default|GetSpawnBGATransform) (Actor|GetOwner)))

(event Collision|EventHit
  (params (MyComp :type Primitive%20Component%20Object%20Reference) (Other :type Actor%20Object%20Reference) (OtherComp :type Primitive%20Component%20Object%20Reference) (bSelfMoved :type Boolean) (HitLocation :type Vector :default "0, 0, 0") (HitNormal :type Vector :default "0, 0, 0") (NormalImpulse :type Vector :default "0, 0, 0") (Hit :type Hit%20Result%20Structure%20%28by%20ref%29))
  ;[1 1 1] On hit, store the Hit Location / Hit Normal then mirror the projectile velocity and hit normal to get the reflection vector
  ;;[1 1 1] If already entered water, don't do anything with this hit. Set hit so it can be checked in water interaction.
  (if (not (Variables|Default|GetEnteredWater))
    (Utilities|FlowControl|DoOnce false
      (:Completed
        (Variables|Default|SetHit true)
        (Variables|Default|SetEnteredWater false)
        (Variables|Default|SetHit%20Normal HitNormal)
        (bind _output_get (Variables|Default|SetHit%20Location HitLocation))
        (Utilities|FlowControl|DoOnce false
          (:Completed
            (Networking|SwitchHasAuthority
              (:Authority
                (Variables|Default|SetSpawnBGATransform (Math|Transform|MakeTransform _output_get))
                (Utilities|Time|SetTimerbyEvent (EventDispatchers|CreateEvent :!event Custom|Defer%20BGA%20Spawn) 0.1)
                (Delivery|DoExplosion))
              (:Remote
                (CallFunction|DoClientExplosionEffects)))))))))

(event Custom|DoClientExplosionEffects
  (bind _hit_location (Variables|Default|GetHit%20Location))
  ;[1 0.6116 0.1752] FX Explosion
  (Utilities|FlowControl|Sequence
    (:then_0
      (Components|Activation|Deactivate (Variables|Default|GetNiagara_Projectile_Particles)))
    (:then_1
      (Effects|Components|ParticleSystem|SpawnFXAtLocation "/ChronoSwarmGrenade/FX/NS_AI_Wildlife_InsectSwarm_Jar_Explosion.NS_AI_Wildlife_InsectSwarm_Jar_Explosion" _hit_location (Math|Conversions|RotationFromXVector (Variables|Default|GetHit%20Normal)) "1.000000,1.000000,1.000000" true "AutoRelease"))
    (:then_3
      (Fortnite|PlayLocalForceFeedbackatLocationMulti "/Game/Weapons/ForceFeedback/FF_Generic_Explosion_Small_Near.FF_Generic_Explosion_Small_Near" "/Game/Weapons/ForceFeedback/FF_Generic_Explosion_Small_Far.FF_Generic_Explosion_Small_Far" _hit_location))
    (:then_4
      (Camera|PlayWorldCameraShake "/Game/Effects/Fort_Effects/BluePrints/CameraShakes/B_Medium_Explosion_CameraShake.B_Medium_Explosion_CameraShake_C" _hit_location))
    (:then_5
      (Fortnite|ShowSoundIndicatorLocal 0 _hit_location (Variables|/Script/Engine.SoundBase|GetMaxDistance (Variables|Sounds|GetCue_Bounce)) "Gunshot")
      (Development|SetHiddeninGame (Variables|B_Prj_Grenade|GetMesh) true false))))

(fn OnRep_EnteredWater
  (bind _returnvalue (Transformation|GetActorLocation))
  (if (Variables|Default|GetEnteredWater)
    (Audio|PlaySoundatLocation (Variables|Sounds|GetSplash%20Sound) _returnvalue)
    (else
      (Audio|PlaySoundatLocation (Variables|Sounds|GetCue_Bounce) _returnvalue))))

(fn ConstructionScript)
